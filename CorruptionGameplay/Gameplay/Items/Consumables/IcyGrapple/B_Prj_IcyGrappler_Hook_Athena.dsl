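; Hook projectile for the Icy Grappler consumable that applies an icy feet gameplay effect to the player pawn when the hook hits.
; The hit event runs the parent projectile's handling first, then a DoOnce limits the effect to the first impact.
; The pawn's ability system component is both source and target, so the struck actor is unaffected by this effect, whose context is built with the projectile.
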(event Collision|EventHit
  (params (MyComp :type Primitive%20Component%20Object%20Reference) (Other :type Actor%20Object%20Reference) (OtherComp :type Primitive%20Component%20Object%20Reference) (bSelfMoved :type Boolean) (HitLocation :type Vector :default "0, 0, 0") (HitNormal :type Vector :default "0, 0, 0") (NormalImpulse :type Vector :default "0, 0, 0") (Hit :type Hit%20Result%20Structure%20%28by%20ref%29))
  (bind _playerpawn (Variables|Default|GetPlayerPawn))
  (bind _returnvalue (Ability|GetAbilitySystemComponent _playerpawn))
  (|CallParentFunction MyComp Other OtherComp bSelfMoved HitLocation HitNormal NormalImpulse Hit)
  (Utilities|FlowControl|DoOnce false
    (:Completed
      (bind _asfort_ability_system_component (Utilities|Casting|CastToFortAbilitySystemComponent _returnvalue))
      (GameplayEffects|ApplyGameplayEffectToTarget _returnvalue (Variables|Default|GetIcy%20feet%20GE%20to%20apply) _returnvalue 0.0 (GameplayEffects|MakeFortEffectContext _asfort_ability_system_component 0 self 0 0)) ; Apply Icy Feet
    )))

(fn ConstructionScript
  (|CallParentFunction))
